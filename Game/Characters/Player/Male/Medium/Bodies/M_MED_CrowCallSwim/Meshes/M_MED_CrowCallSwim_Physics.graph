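# Physics asset for the M_MED_CrowCallSwim male medium player body, covering a two-bone necklace and the chest around it.
# The necklace capsules chain from a small spine_05 body through constraints with limited swing and locked twist.
# Three larger capsules on center, left and right chest bones approximate the torso, likely as necklace colliders, and join spine_05 with free angular motion.
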PhysicsAsset /Game/Characters/Player/Male/Medium/Bodies/M_MED_CrowCallSwim/Meshes/M_MED_CrowCallSwim_Physics
bodies {
 base_chest [{shapeType=Capsule, length=13.0843, radius=10.3371, shapeName=base_chest_capsule, center=<-4.0494 11.1343 0>, rotation=<rot 0 -25.5499 0>}]
 base_chest_l [{
   shapeType=Capsule
   length=5.5527
   radius=6.4168
   shapeName=base_chest_l_capsule
   center=<-0.894 6.4735 0.4654>
   rotation=<rot 2.0237 -86.9976 19.4176>
  }]
 base_chest_r [{
   shapeType=Capsule
   length=5.5527
   radius=6.4168
   shapeName=base_chest_r_capsule
   center=<-0.8714 6.5076 -0.4642>
   rotation=<rot 2.0237 86.9916 19.4176>
  }]
 necklace_01 [{shapeType=Capsule, length=2, radius=1, shapeName=necklace_01_capsule, center=<1.9972 0 0.5366>, rotation=<rot 88.4967 180 180>}]
 necklace_02 [{shapeType=Capsule, length=1.9293, radius=1.0707, shapeName=necklace_02_capsule, center=<1.9935 0 0.5419>, rotation=<rot 87.3413 0 0>}]
 spine_05 [{shapeType=Capsule, length=0.5, radius=0.5, shapeName=spine_05_capsule}]
}
constraints [
 {bone1Name=necklace_02, bone2Name=necklace_01, swing1LimitDegrees=15, swing2LimitDegrees=20, twistLimitDegrees=45, twistMotion=Locked}
 {bone1Name=necklace_01, bone2Name=spine_05, swing1LimitDegrees=15, swing2LimitDegrees=20, twistLimitDegrees=45, twistMotion=Locked}
 {
  bone1Name=base_chest
  bone2Name=spine_05
  swing1LimitDegrees=45
  swing1Motion=Free
  swing2LimitDegrees=45
  swing2Motion=Free
  twistLimitDegrees=45
  twistMotion=Free
 }
 {
  bone1Name=base_chest_l
  bone2Name=spine_05
  swing1LimitDegrees=45
  swing1Motion=Free
  swing2LimitDegrees=45
  swing2Motion=Free
  twistLimitDegrees=45
  twistMotion=Free
 }
 {
  bone1Name=base_chest_r
  bone2Name=spine_05
  swing1LimitDegrees=45
  swing1Motion=Free
  swing2LimitDegrees=45
  swing2Motion=Free
  twistLimitDegrees=45
  twistMotion=Free
 }
]
