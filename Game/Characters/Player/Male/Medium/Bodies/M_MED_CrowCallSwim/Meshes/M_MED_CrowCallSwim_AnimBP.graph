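# Post-processes the pose of the CrowCallSwim medium male player body, which follows its attached parent mesh, to drive secondary gear bones.
# The copied pose runs through a Control Rig, weighted chest constraints, a small chest offset during 2D movement and RigidBody simulation, then gets cached.
# Weighted multi-bone constraints then position bag, gun, knee pad and bullet bones.
# EventGraph updates the 2D-movement alpha and RigidBody gravity override, the latter via a deprecated function.
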AnimBlueprint /Game/Characters/Player/Male/Medium/Bodies/M_MED_CrowCallSwim/Meshes/M_MED_CrowCallSwim_AnimBP
target_skeleton=&/M_MED_CrowCallSwim_Skeleton
graphs [
 AnimGraph {nodes [
   n1 = "Constraint - base_bag_01_l (PPPPPPPP)"!(
    ComponentPose=Local To Component(LocalPose=Use cached pose 'CR_Start'())
    Alpha=1
    ConstraintWeights_0=0.6068
    ConstraintWeights_1:float=0.2739
    ConstraintWeights_2:float=0.0344
    ConstraintWeights_3:float=0.0274
    ConstraintWeights_4:float=0.0215
    ConstraintWeights_5:float=0.0183
    ConstraintWeights_6:float=0.0168
    ConstraintWeights_7:float=0.001
   ) -> (Pose) {
    alphaBoolBlend {blendOption=Linear}
    boneToModify {boneName=base_bag_01_l}
    constraintSetup [
     {targetBone {boneName=dyn_belt_l}}
     {targetBone {boneName=pelvis}}
     {targetBone {boneName=spine_02}}
     {targetBone {boneName=spine_01}}
     {targetBone {boneName=spine_03}}
     {targetBone {boneName=thigh_twist_01_l}}
     {targetBone {boneName=deform_groin_l}}
     {targetBone {boneName=spine_04}}
    ]
    constraintWeights [1,1,1,1,1,1,1,1]
   }
   n2 = "Constraint - base_bag_02_l (PPPPPPPP)"!(
    ComponentPose=@n1
    Alpha=1
    ConstraintWeights_0=0.6226
    ConstraintWeights_1:float=0.2642
    ConstraintWeights_2:float=0.0319
    ConstraintWeights_3:float=0.0253
    ConstraintWeights_4:float=0.0243
    ConstraintWeights_5:float=0.0195
    ConstraintWeights_6:float=0.0097
    ConstraintWeights_7:float=0.0024
   ) -> (Pose) {
    alphaBoolBlend {blendOption=Linear}
    boneToModify {boneName=base_bag_02_l}
    constraintSetup [
     {targetBone {boneName=dyn_belt_l}}
     {targetBone {boneName=pelvis}}
     {targetBone {boneName=spine_02}}
     {targetBone {boneName=spine_01}}
     {targetBone {boneName=thigh_twist_01_l}}
     {targetBone {boneName=spine_03}}
     {targetBone {boneName=deform_groin_l}}
     {targetBone {boneName=deform_glute_l}}
    ]
    constraintWeights [1,1,1,1,1,1,1,1]
   }
   n3 = "Constraint - base_bag_01_r (PPPPPPPP)"!(
    ComponentPose=@n2
    Alpha=1
    ConstraintWeights_0=0.6068
    ConstraintWeights_1:float=0.2739
    ConstraintWeights_2:float=0.0344
    ConstraintWeights_3:float=0.0274
    ConstraintWeights_4:float=0.0215
    ConstraintWeights_5:float=0.0183
    ConstraintWeights_6:float=0.0168
    ConstraintWeights_7:float=0.001
   ) -> (Pose) {
    alphaBoolBlend {blendOption=Linear}
    boneToModify {boneName=base_bag_01_r}
    constraintSetup [
     {targetBone {boneName=dyn_belt_r}}
     {targetBone {boneName=pelvis}}
     {targetBone {boneName=spine_02}}
     {targetBone {boneName=spine_01}}
     {targetBone {boneName=spine_03}}
     {targetBone {boneName=thigh_twist_01_r}}
     {targetBone {boneName=deform_groin_r}}
     {targetBone {boneName=spine_04}}
    ]
    constraintWeights [1,1,1,1,1,1,1,1]
   }
   n4 = "Constraint - base_bag_02_r (PPPPPPPP)"!(
    ComponentPose=@n3
    Alpha=1
    ConstraintWeights_0=0.6226
    ConstraintWeights_1:float=0.2642
    ConstraintWeights_2:float=0.0319
    ConstraintWeights_3:float=0.0253
    ConstraintWeights_4:float=0.0243
    ConstraintWeights_5:float=0.0195
    ConstraintWeights_6:float=0.0097
    ConstraintWeights_7:float=0.0024
   ) -> (Pose) {
    alphaBoolBlend {blendOption=Linear}
    boneToModify {boneName=base_bag_02_r}
    constraintSetup [
     {targetBone {boneName=dyn_belt_r}}
     {targetBone {boneName=pelvis}}
     {targetBone {boneName=spine_02}}
     {targetBone {boneName=spine_01}}
     {targetBone {boneName=thigh_twist_01_r}}
     {targetBone {boneName=spine_03}}
     {targetBone {boneName=deform_groin_r}}
     {targetBone {boneName=deform_glute_r}}
    ]
    constraintWeights [1,1,1,1,1,1,1,1]
   }
   n5 = "Constraint - base_gun_l (PPPPP)"!(
    ComponentPose=@n4
    Alpha=1
    ConstraintWeights_0=0.8398
    ConstraintWeights_1:float=0.0863
    ConstraintWeights_2:float=0.0438
    ConstraintWeights_3:float=0.0225
    ConstraintWeights_4:float=0.0075
   ) -> (Pose) {
    alphaBoolBlend {blendOption=Linear}
    boneToModify {boneName=base_gun_l}
    constraintSetup [
     {targetBone {boneName=thigh_twist_01_l}}
     {targetBone {boneName=pelvis}}
     {targetBone {boneName=deform_groin_l}}
     {targetBone {boneName=thigh_l}}
     {targetBone {boneName=deform_glute_l}}
    ]
    constraintWeights [1,1,1,1,1]
   }
   n6 = "Constraint - base_gun_r (PPPPP)"!(
    ComponentPose=@n5
    Alpha=1
    ConstraintWeights_0=0.8398
    ConstraintWeights_1:float=0.0863
    ConstraintWeights_2:float=0.0438
    ConstraintWeights_3:float=0.0225
    ConstraintWeights_4:float=0.0075
   ) -> (Pose) {
    alphaBoolBlend {blendOption=Linear}
    boneToModify {boneName=base_gun_r}
    constraintSetup [
     {targetBone {boneName=thigh_twist_01_r}}
     {targetBone {boneName=pelvis}}
     {targetBone {boneName=deform_groin_r}}
     {targetBone {boneName=thigh_r}}
     {targetBone {boneName=deform_glute_r}}
    ]
    constraintWeights [1,1,1,1,1]
   }
   n7 = "Constraint - base_kneePad_l (PPPPPP)"!(
    ComponentPose=@n6
    Alpha=1
    ConstraintWeights_0=0.3208
    ConstraintWeights_1:float=0.2818
    ConstraintWeights_2:float=0.1939
    ConstraintWeights_3:float=0.1608
    ConstraintWeights_4:float=0.0339
    ConstraintWeights_5:float=0.0089
   ) -> (Pose) {
    alphaBoolBlend {blendOption=Linear}
    boneToModify {boneName=base_kneePad_l}
    constraintSetup [
     {targetBone {boneName=thigh_l}}
     {targetBone {boneName=deform_knee_out_l}}
     {targetBone {boneName=calf_twist_02_l}}
     {targetBone {boneName=calf_l}}
     {targetBone {boneName=thigh_twist_01_l}}
     {targetBone {boneName=deform_knee_in_l}}
    ]
    constraintWeights [1,1,1,1,1,1]
   }
   n8 = "Constraint - base_kneePad_r (PPPPPP)"!(
    ComponentPose=@n7
    Alpha=1
    ConstraintWeights_0=0.3208
    ConstraintWeights_1:float=0.2818
    ConstraintWeights_2:float=0.1939
    ConstraintWeights_3:float=0.1608
    ConstraintWeights_4:float=0.0339
    ConstraintWeights_5:float=0.0089
   ) -> (Pose) {
    alphaBoolBlend {blendOption=Linear}
    boneToModify {boneName=base_kneePad_r}
    constraintSetup [
     {targetBone {boneName=thigh_r}}
     {targetBone {boneName=deform_knee_out_r}}
     {targetBone {boneName=calf_twist_02_r}}
     {targetBone {boneName=calf_r}}
     {targetBone {boneName=thigh_twist_01_r}}
     {targetBone {boneName=deform_knee_in_r}}
    ]
    constraintWeights [1,1,1,1,1,1]
   }
   n9 = "Constraint - base_bullets (PPPPPPP)"!(
    ComponentPose=@n8
    Alpha=1
    ConstraintWeights_0=0.6408
    ConstraintWeights_1:float=0.1608
    ConstraintWeights_2:float=0.1088
    ConstraintWeights_3:float=0.0748
    ConstraintWeights_4:float=0.0105
    ConstraintWeights_5:float=0.0033
    ConstraintWeights_6:float=0.0009
   ) -> (Pose) {
    alphaBoolBlend {blendOption=Linear}
    boneToModify {boneName=base_bullets}
    constraintSetup [
     {targetBone {boneName=thigh_twist_01_r}}
     {targetBone {boneName=deform_groin_r}}
     {targetBone {boneName=pelvis}}
     {targetBone {boneName=thigh_r}}
     {targetBone {boneName=thigh_twist_01_l}}
     {targetBone {boneName=deform_groin_l}}
     {targetBone {boneName=thigh_l}}
    ]
    constraintWeights [1,1,1,1,1,1,1]
   }
   Output Pose(Result=Component To Local(ComponentPose=@n9)) {layerGroup=None}
   n10 = Control Rig(Source=Copy Pose From Mesh() {bUseAttachedParent=true}) {
    alphaBoolBlend {blendOption=Linear}
    controlRigAssetReference {blueprintRigClass=&/M_MED_CrowCallSwim_CtrlRig.M_MED_CrowCallSwim_CtrlRig_C}
   }
   n11 = "Constraint - base_chest (PPPPPPPP)"!(
    ComponentPose=Local To Component(LocalPose=@n10)
    Alpha=1
    ConstraintWeights_0=0.8365
    ConstraintWeights_1:float=0.0786
    ConstraintWeights_2:float=0.0445
    ConstraintWeights_3:float=0.0183
    ConstraintWeights_4:float=0.0093
    ConstraintWeights_5:float=0.0093
    ConstraintWeights_6:float=0.0018
    ConstraintWeights_7:float=0.0018
   ) -> (Pose) {
    alphaBoolBlend {blendOption=Linear}
    boneToModify {boneName=base_chest}
    constraintSetup [
     {targetBone {boneName=spine_05}}
     {targetBone {boneName=spine_04}}
     {targetBone {boneName=spine_03}}
     {targetBone {boneName=spine_02}}
     {targetBone {boneName=clavicle_l}}
     {targetBone {boneName=clavicle_r}}
     {targetBone {boneName=deform_pec_l}}
     {targetBone {boneName=deform_pec_r}}
    ]
    constraintWeights [1,1,1,1,1,1,1,1]
   }
   n12 = "Constraint - base_chest_l (PPPPPPPP)"!(
    ComponentPose=@n11
    Alpha=1
    ConstraintWeights_0=0.3954
    ConstraintWeights_1:float=0.229
    ConstraintWeights_2:float=0.1361
    ConstraintWeights_3:float=0.1321
    ConstraintWeights_4:float=0.0499
    ConstraintWeights_5:float=0.0413
    ConstraintWeights_6:float=0.0104
    ConstraintWeights_7:float=0.0057
   ) -> (Pose) {
    alphaBoolBlend {blendOption=Linear}
    boneToModify {boneName=base_chest_l}
    constraintSetup [
     {targetBone {boneName=spine_05}}
     {targetBone {boneName=spine_04}}
     {targetBone {boneName=spine_03}}
     {targetBone {boneName=clavicle_l}}
     {targetBone {boneName=spine_02}}
     {targetBone {boneName=deform_pec_l}}
     {targetBone {boneName=deform_upperarm_fr_l}}
     {targetBone {boneName=upperarm_twist_01_l}}
    ]
    constraintWeights [1,1,1,1,1,1,1,1]
   }
   n13 = "Constraint - base_chest_r (PPPPPPPP)"!(
    ComponentPose=@n12
    Alpha=1
    ConstraintWeights_0=0.3954
    ConstraintWeights_1:float=0.229
    ConstraintWeights_2:float=0.1361
    ConstraintWeights_3:float=0.1321
    ConstraintWeights_4:float=0.0499
    ConstraintWeights_5:float=0.0413
    ConstraintWeights_6:float=0.0104
    ConstraintWeights_7:float=0.0057
   ) -> (Pose) {
    alphaBoolBlend {blendOption=Linear}
    boneToModify {boneName=base_chest_r}
    constraintSetup [
     {targetBone {boneName=spine_05}}
     {targetBone {boneName=spine_04}}
     {targetBone {boneName=spine_03}}
     {targetBone {boneName=clavicle_r}}
     {targetBone {boneName=spine_02}}
     {targetBone {boneName=deform_pec_r}}
     {targetBone {boneName=deform_upperarm_fr_r}}
     {targetBone {boneName=upperarm_twist_01_r}}
    ]
    constraintWeights [1,1,1,1,1,1,1,1]
   }
   n14 = "Transform (Modify) Bone - Bone: base_chest"(ComponentPose=@n13, Alpha=Get IsMoving2D() -> (IsMoving2D:double), Translation="0,-0.7,0", Rotation="0,0,0", Scale="1,1,1") {
    translationMode=BMM_Additive
    translationSpace=BCS_BoneSpace
    alphaBoolBlend {blendOption=Linear}
    boneToModify {boneName=base_chest}
    translation {y=-0.7}
   }
   n15 = RigidBody!(ComponentPose=@n14, Alpha, OverrideWorldGravity:Vector=Get Gravity Override(), ExternalForce) -> (Pose) {
    bOverrideWorldGravity=true
    overridePhysicsAsset=&/M_MED_CrowCallSwim_Physics
    simulationSpace=BaseBoneSpace
    alphaBoolBlend {blendOption=Linear}
    alphaScaleBiasClamp {bInterpResult=true}
    baseBoneRef {boneName=spine_03}
    componentAppliedLinearAccClamp=<3000 3000 3000>
    componentLinearAccScale=<0.3 0.3 0.3>
    componentLinearVelScale=<0.3 0.3 0.3>
    simSpaceSettings {
     maxAngularAcceleration=1000
     maxAngularVelocity=3
     maxLinearAcceleration=700
     maxLinearVelocity=500
     velocityScaleZ=0
     worldAlpha=0.7
     externalLinearDragV=<0.2 0.2>
    }
   }
   Save cached pose 'CR_Start'(Pose=Component To Local(ComponentPose=@n15)) {cacheName=CR_Start}
  ]}
 EventGraph {nodes [
   Event BlueprintUpdateAnimation()
   > n1 = Sequence()
   > n2 = Gravity Override State Function DEPRECATED(CharacterPartAnimInstance=Self-Reference(), GameTimeInSeconds=GetGameTimeInSeconds(), GravityOverrideParams=Get Gravity Override Params())
   > Set Gravity Override(Gravity Override=@n2)
   n3 = Branch(execute=@n1.then_1, Condition=Get bIsMoving2D())
   > Set IsMoving2D(IsMoving2D:double=1) -> (Output_Get:double)
   Set IsMoving2D(execute=@n3.else, IsMoving2D=0)
  ]}
]
variables [
 Gravity Override {type=Vector}
 Gravity Override Params {type=GravityOverrideParamsStruct}
 IsMoving2D {type=number}
]
